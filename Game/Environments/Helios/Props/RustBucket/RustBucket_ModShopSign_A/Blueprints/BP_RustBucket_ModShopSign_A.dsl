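; Rust Bucket mod shop sign prop Blueprint that tints its emissive glow from a global lighting color modifier.
; The construction script runs the parent script, then writes that color to the Emissive_Color_Multiplier parameter on the static mesh component.
; It sets the value both on a dynamic material instance from BTechArtGlobalLibrary and across the component's materials as a vector.
; BeginPlay, overlap and Tick events are empty.
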(event EventBeginPlay)

(event Collision|EventActorBeginOverlap
  (params (OtherActor :type Actor%20Object%20Reference)))

(event EventTick
  (params (DeltaSeconds :type Float%20%28single-precision%29)))

(fn ConstructionScript
  (bind _emissive_color_modifier (Variables|00_GlobalLighting|GetEmissive%20Color%20Modifier))
  (|CallParentFunction)
  (Rendering|Material|SetVectorParameterValue (Class|BTechArtGlobalLibrary|GetDynamicMaterialInstance (Variables|Debug|GetStaticMeshComponent)) "Emissive_Color_Multiplier" _emissive_color_modifier :!class /Script/Engine.MaterialInstanceDynamic)
  (Rendering|Material|SetVectorParameterValueonMaterials (Variables|Debug|GetStaticMeshComponent) "Emissive_Color_Multiplier" (Math|Conversions|ToVector(LinearColor) _emissive_color_modifier)))
